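# Translucent unlit Niagara screen-distortion material that refracts and blurs scene color with colored fringing.
# A panning, rotating flow map and noisy radial streaks around the particle's screen position build refraction and blur vectors.
# Six scene-color taps along the blur, with dither declared for jitter, are tinted by a three-color aberration gradient and averaged.
# Dynamic parameters drive strengths, dissolve mask and fades; PC-Perf, Android and Switch output zero.
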Material /BattlePassS36_QuestActors/VFX/Materials/M_VoidMask_Blur_Copied
graph {
 properties {
  bAutomaticallySetUsageInEditor=false
  bEnableResponsiveAA=true
  bUseTranslucencyVertexFog=false
  bUsedWithNiagaraMeshParticles=true
  bUsedWithNiagaraSprites=true
  blendMode=BLEND_Translucent
  shadingModel=MSM_Unlit
  translucencyPass=MTP_BeforeDOF
  twoSided=true
  usedShadingModels=Unlit
 }
 nodes [
  n1 = DynamicParameter() {
   parameterIndex=1
   defaultValue=<rgba 0 200 1 1>
   paramNames [Desaturation, DistanceFade, VelocityTextureRotationSpeed, UseSpriteUVForVelocityTexture]
  }
  n2 = DynamicParameter() {parameterIndex=3, defaultValue=<rgba 1 200 0.7 1>, paramNames [RefactionStrength, CameraFade, Param3, Param4]}
  n3 = Multiply(
   A=Multiply(A=NamedRerouteUsage(), B=DepthFade(FadeDistance=@n1.DistanceFade) {fadeDistanceDefault=200})
   B=Power(Base=MaterialFunctionCall(Distance=@n2.CameraFade) {materialFunction=&Near_Camera_Fade}.Result)
  )
  n4 = Constant()
  MaterialFunctionCall("Xb1,PS4"=@n3, "PC-High&&Epic,XSX,PS5"=@n3, "PC-Perf,Android"=@n4, Switch=@n4, "PC-Low(NotPerf)&&Medium"=@n3) {materialFunction=&PlatformSwitch_1D}
  n5 = NamedRerouteUsage()
  n6 = ParticleColor()
  n7 = NamedRerouteUsage()
  n8 = NamedRerouteUsage()
  n9 = LinearInterpolate(A=@n5, B=Multiply(A=@n5, B=@n6.RGB), Alpha=Multiply(A=Multiply(A=@n7, B=@n8), B=@n6.A))
  MaterialFunctionCall!("Xb1,PS4"=@n9, "PC-High&&Epic,XSX,PS5"=@n9, "PC-Perf,Android"=@n4, Switch=@n4, "PC-Low(NotPerf)&&Medium"=@n9) {materialFunction=/Game/Effects/Niagara/MasterMaterials/Functions/Common/PlatformSwitch_3D}
  n10 = NamedRerouteUsage()
  n11 = NamedRerouteUsage()
  n12 = Multiply(
   A=Multiply(A=Add(A=@n10, B=@n11), B=ScalarParameter() {defaultValue=0.5, group="####!QuickParam", parameterName=TotalRefractionScale})
   B=NamedRerouteUsage()
  )
  n13 = NamedRerouteUsage()
  NamedRerouteDeclaration(None=Multiply(A=Multiply(A=@n12, B=@n13), B=@n2.RefactionStrength)) {name=Refrac, nodeColor=<rgba 0 1 0.9766 1>}
  n14 = Multiply(
   A=Add(A=NamedRerouteUsage(), B=NamedRerouteUsage())
   B=ScalarParameter() {defaultValue=1, group="####!QuickParam", parameterName=TotalBlurStrScale}
  )
  n15 = NamedRerouteUsage()
  NamedRerouteDeclaration(None=Multiply(A=Multiply(A=@n14, B=@n15), B=NamedRerouteUsage())) {name=UniformFlow, nodeColor=<rgba 0 0.7891 1 1>}
  n16 = TextureCoordinate()
  n17 = Constant2Vector() {g=0.5, r=0.5}
  n18 = DynamicParameter() {defaultValue=<rgba 1 1 0.7 1>, paramNames [VelocityTextureStrength, RadialBlurStrength, MaskDissolve, UseMask]}
  n19 = Power(Base=OneMinus(None=SphereMask(A=@n16, B=@n17) {attenuationRadius=0.5, hardnessPercent=0}), Exp=@n18.MaskDissolve)
  n20 = ScalarParameter() {defaultValue=2, group=mask, parameterName=mask_DissolveStr}
  n21 = TextureObjectParameter() {
   group=mask
   parameterName=mask_Texture
   samplerType=SAMPLERTYPE_LinearGrayscale
   sortPriority=0
   texture=/Game/Effects/Fort_Effects/Emotes/UltraLight/Texture/T_FluidFlameTile_Greyscale
  }
  n22 = TextureCoordinate()
  n23 = VectorParameter() {
   group=mask
   parameterName=mask_NoiseFromVelocityTexControl
   sortPriority=60
   channelNames {a=VStrength, b=UStrength, g=VOffset, r=UOffset}
   defaultValue=<rgba 0 0 0.1 0.1>
  }
  n24 = VectorParameter() {group=mask, parameterName=mask_UVControl, channelNames {a=VPan, b=UPan, g=VScale, r=UScale}, defaultValue=<rgba 3 0.2 0 -2>}
  n25 = DynamicParameter() {
   parameterIndex=2
   defaultValue=<rgba 1 1 0.7 1>
   paramNames [VelocityTextureTime, RadialBlurStrengthTime, MaskDissolveTime, ScaleVelocityTextureUV]
  }
  n26 = MaterialFunctionCall!(
   Texture=@n21
   "[Override]DDXInput"=DDX(Value=@n22)
   "[Override]DDYInput"=DDY(Value=@n22)
   AdditionalUVOffset=MaterialFunctionCall!(NoiseValue=NamedRerouteUsage(), Control_RGB=@n23, Control_A=@n23) {materialFunction=/Game/Effects/Niagara/MasterMaterials/Functions/Common/FXCommon_NoiseControl}.NoiseValue
   UVControl_RGB=@n24
   UVControl_A=@n24
   "[Override]TimeInput"=@n25.MaskDissolveTime
   "[Override]UVInput"=NamedRerouteUsage()
  ) {materialFunction=/Game/Effects/Niagara/MasterMaterials/Functions/Common/FXCommon_TextureSample_DDXDDY}
  n27 = Subtract(B=Multiply(A=Multiply(A=@n19, B=@n20), B=OneMinus(None=@n26.R)))
  n28 = ScalarParameter() {defaultValue=3, group=mask, parameterName=mask_GrayAdjust_Mul}
  n29 = Power(Base=Saturate(None=Multiply(A=@n27, B=@n28)), Exp=ScalarParameter() {defaultValue=1, group=mask, parameterName=mask_GrayAdjust_Power})
  n30 = DynamicParameter() {defaultValue=<rgba 1 1 0.7 1>, paramNames [VelocityTextureStrength, RadialBlurStrength, MaskDissolve, UseMask]}
  NamedRerouteDeclaration(None=LinearInterpolate(B=Saturate(None=@n29), Alpha=@n30.UseMask) {constA=1}) {name=MaskT, nodeColor=<rgba 0 1 0.25 1>}
  n31 = TextureCoordinate()
  n32 = Constant2Vector() {g=0.5, r=0.5}
  n33 = ScalarParameter() {defaultValue=0.2, group="####!QuickParam", parameterName=TotalSphereMask_Hardness}
  n34 = ScalarParameter() {defaultValue=0.8, parameterName=CenterMask_Offset}
  n35 = Subtract(A=SphereMask(A=@n31, B=@n32) {attenuationRadius=0.5, hardnessPercent=0}, B=@n34)
  n36 = Power(Base=Divide(A=Saturate(None=@n35), B=OneMinus(None=@n34)), Exp=ScalarParameter() {defaultValue=2, parameterName=CenterMask_Power})
  NamedRerouteDeclaration(None=Multiply(A=SphereMask(A=@n31, B=@n32, Hardness=@n33) {attenuationRadius=0.5, hardnessPercent=0.2}, B=OneMinus(None=@n36))) {name=SphereMask, nodeColor=<rgba 1 0 0.0703 1>}
  n37 = NamedRerouteUsage()
  n38 = TextureObjectParameter() {
   group=TextureRadialBlur
   parameterName=TextureRadialBlur_Texture
   samplerType=SAMPLERTYPE_LinearGrayscale
   sortPriority=0
   texture=&T_SimplePerlin_Greyscale
  }
  n39 = TextureCoordinate()
  n40 = VectorParameter() {
   group=TextureRadialBlur
   parameterName=TextureRadialBlur_UVControl
   channelNames {a=VPan, b=UPan, g=VScale, r=UScale}
   defaultValue=<rgba 2 1 0 -0.5>
  }
  n41 = DynamicParameter() {
   parameterIndex=2
   defaultValue=<rgba 1 1 0.7 1>
   paramNames [VelocityTextureTime, RadialBlurStrengthTime, MaskDissolveTime, ScaleVelocityTextureUV]
  }
  n42 = TextureObjectParameter() {
   group=TextureRadialBlur
   parameterName=TextureRadialBlurNoise_Texture
   samplerType=SAMPLERTYPE_LinearGrayscale
   sortPriority=0
   texture=/Game/Effects/Fort_Effects/Emotes/UltraLight/Texture/T_Plasma_A_B_Greyscale
  }
  n43 = TextureCoordinate()
  n44 = VectorParameter() {
   group=TextureRadialBlur
   parameterName=TextureRadialBlurNoise_UVControl
   channelNames {a=VPan, b=UPan, g=VScale, r=UScale}
   defaultValue=<rgba 1 0.05 0 -3>
  }
  n45 = NamedRerouteUsage()
  n46 = VectorParameter() {
   group=TextureRadialBlur
   parameterName=TextureRadialBlurNoise_NoiseControl
   sortPriority=60
   channelNames {a=VStrength, b=UStrength, g=VOffset, r=UOffset}
   defaultValue=<rgba 0 0 0 0.4>
  }
  n47 = MaterialFunctionCall(
   NoiseValue=MaterialFunctionCall(
    Texture=@n42
    "[Override]DDXInput"=DDX(Value=@n43)
    "[Override]DDYInput"=DDY(Value=@n43)
    UVControl_RGB=@n44
    UVControl_A=@n44
    "[Override]TimeInput"=@n41.RadialBlurStrengthTime
    "[Override]UVInput"=@n45
   ) {materialFunction=&FXCommon_TextureSample_DDXDDY}.RG
   Control_RGB=@n46
   Control_A=@n46
  ) {materialFunction=&FXCommon_NoiseControl}
  n48 = Power(
   Base=MaterialFunctionCall(
    Texture=@n38
    "[Override]DDXInput"=DDX(Value=@n39)
    "[Override]DDYInput"=DDY(Value=@n39)
    UVControl_RGB=@n40
    UVControl_A=@n40
    "[Override]TimeInput"=@n41.RadialBlurStrengthTime
    "[Override]UVInput"=Add(A=@n47.NoiseValue, B=@n45)
   ) {materialFunction=&FXCommon_TextureSample_DDXDDY}.R
   Exp=ScalarParameter() {defaultValue=1, parameterName=TextureRadialBlur_ValuePower}
  )
  n49 = Multiply(
   A=MaterialFunctionCall(Input=@n48, Target Low=ScalarParameter() {group=TextureRadialBlur, parameterName=TextureRadialBlur_ToBlurRemapLowestValue}) {materialFunction=&RemapValueRange}.Result
   B=ScalarParameter() {defaultValue=0.4, group=TextureRadialBlur, parameterName=TextureRadialBlur_ToBlurStr}
  )
  n50 = DynamicParameter() {defaultValue=<rgba 1 1 0.7 1>, paramNames [VelocityTextureStrength, RadialBlurStrength, MaskDissolve, UseMask]}
  NamedRerouteDeclaration(None=Multiply(A=Multiply(A=@n37, B=@n49), B=@n50.RadialBlurStrength)) {name=TextureRadialBlur_ToBlur, nodeColor=<rgba 0 0.3906 1 1>}
  n51 = ScalarParameter() {defaultValue=0.3, group=TextureRadialBlur, parameterName=TextureRadialBlur_ToRefractionStr}
  n52 = Multiply(
   A=MaterialFunctionCall(
    Input=@n48
    Target Low=ScalarParameter() {defaultValue=0.4, group=TextureRadialBlur, parameterName=TextureRadialBlur_ToRefractionRemapLowestValue}
   ) {materialFunction=&RemapValueRange}.Result
   B=NamedRerouteUsage()
  )
  NamedRerouteDeclaration(None=Multiply(A=@n50.RadialBlurStrength, B=Multiply(A=@n51, B=@n52))) {name=TextureRadialBlur_ToRefraction, nodeColor=<rgba 0.7109 1 0 1>}
  n53 = NamedRerouteUsage()
  n54 = SceneColor!(OffsetFraction=@n53) {inputMode=OffsetFraction}
  n55 = DynamicParameter() {
   parameterIndex=1
   defaultValue=<rgba 0 200 1 1>
   paramNames [Desaturation, DistanceFade, VelocityTextureRotationSpeed, UseSpriteUVForVelocityTexture]
  }
  n56 = ParticleColor()
  n57 = LinearInterpolate(B=Multiply(A=@n55.Desaturation, B=@n56.A), Alpha=NamedRerouteUsage())
  n58 = Desaturation(None=@n54.RGB, Fraction=@n57)
  n59 = NamedRerouteUsage()
  n60 = VectorParameter() {parameterName=AberrationColor_0, defaultValue=<rgba 1 0 0 1>}
  n61 = NamedRerouteUsage()
  n62 = Frac(None=Constant() {r=0.2})
  n63 = NamedRerouteUsage()
  n64 = MaterialFunctionCall(Input=Add(A=Add(A=@n62, B=@n63) {constB=0.5}) {constB=-0.5}, Target Low=Frac(None=Constant()), Target High=@n62) {materialFunction=&RemapValueRange}
  n65 = SceneColor(OffsetFraction=Add(A=Multiply(A=@n61, B=@n64.Result) {constB=0.2}, B=@n53)) {inputMode=OffsetFraction}
  n66 = LinearInterpolate(A=@n65.RGB, B=Desaturation(None=@n65.RGB, Fraction=@n57), Alpha=NamedRerouteUsage())
  n67 = VectorParameter() {parameterName=AberrationColor_0.5, defaultValue=<rgba 1 0.3333 0 1>}
  n68 = VectorParameter() {parameterName=AberrationColor_1, defaultValue=<rgba 0 0 1 1>}
  n69 = MaterialFunctionCall(A=@n60.RGB, B=@n67.RGB, C=@n68.RGB, Alpha=Constant() {r=0.2}) {materialFunction=&Lerp_3Color}
  n70 = Frac(None=Constant() {r=0.4})
  n71 = NamedRerouteUsage()
  n72 = MaterialFunctionCall(Input=Add(A=Add(A=@n70, B=@n71) {constB=0.5}) {constB=-0.5}, Target Low=@n62, Target High=@n70) {materialFunction=&RemapValueRange}
  n73 = SceneColor(OffsetFraction=Add(A=Multiply(A=@n61, B=@n72.Result) {constB=0.4}, B=@n53)) {inputMode=OffsetFraction}
  n74 = LinearInterpolate(A=@n73.RGB, B=Desaturation(None=@n73.RGB, Fraction=@n57), Alpha=NamedRerouteUsage())
  n75 = MaterialFunctionCall(A=@n60.RGB, B=@n67.RGB, C=@n68.RGB, Alpha=Constant() {r=0.5}) {materialFunction=&Lerp_3Color}
  n76 = Add(
   A=Multiply(A=@n60.RGB, B=LinearInterpolate(A=@n54.RGB, B=@n58, Alpha=@n59))
   B=Add(A=Multiply(A=@n66, B=@n69.Result), B=Multiply(A=@n74, B=@n75.Result))
  )
  n77 = Frac(None=Constant() {r=0.6})
  n78 = NamedRerouteUsage()
  n79 = MaterialFunctionCall(Input=Add(A=Add(A=@n77, B=@n78) {constB=0.5}) {constB=-0.5}, Target Low=@n70, Target High=@n77) {materialFunction=&RemapValueRange}
  n80 = SceneColor(OffsetFraction=Add(A=Multiply(A=@n61, B=@n79.Result) {constB=0.6}, B=@n53)) {inputMode=OffsetFraction}
  n81 = LinearInterpolate(A=@n80.RGB, B=Desaturation(None=@n80.RGB, Fraction=@n57), Alpha=NamedRerouteUsage())
  n82 = MaterialFunctionCall(A=@n60.RGB, B=@n67.RGB, C=@n68.RGB, Alpha=Constant() {r=0.7}) {materialFunction=&Lerp_3Color}
  n83 = Frac(None=Constant() {r=0.8})
  n84 = NamedRerouteUsage()
  n85 = MaterialFunctionCall(Input=Add(A=Add(A=@n83, B=@n84) {constB=0.5}) {constB=-0.5}, Target Low=@n77, Target High=@n83) {materialFunction=&RemapValueRange}
  n86 = SceneColor(OffsetFraction=Add(A=Multiply(A=@n61, B=@n85.Result) {constB=0.8}, B=@n53)) {inputMode=OffsetFraction}
  n87 = LinearInterpolate(A=@n86.RGB, B=Desaturation(None=@n86.RGB, Fraction=@n57), Alpha=NamedRerouteUsage())
  n88 = MaterialFunctionCall(A=@n60.RGB, B=@n67.RGB, C=@n68.RGB, Alpha=Constant() {r=0.9}) {materialFunction=&Lerp_3Color}
  n89 = Frac(None=Constant() {r=1})
  n90 = NamedRerouteUsage()
  n91 = MaterialFunctionCall(Input=Add(A=Add(A=@n89, B=@n90) {constB=0.5}) {constB=-0.5}, Target Low=@n83, Target High=@n89) {materialFunction=&RemapValueRange}
  n92 = SceneColor(OffsetFraction=Add(A=Multiply(A=@n61, B=@n91.Result), B=@n53)) {inputMode=OffsetFraction}
  n93 = Desaturation(None=@n92.RGB, Fraction=@n57)
  n94 = NamedRerouteUsage()
  n95 = Add(
   A=Add(A=Multiply(A=@n81, B=@n82.Result), B=Multiply(A=@n87, B=@n88.Result))
   B=Multiply(A=LinearInterpolate(A=@n92.RGB, B=@n93, Alpha=@n94), B=@n68.RGB)
  )
  n96 = Add(A=@n69.Result, B=@n75.Result)
  n97 = Add(A=Add(A=Add(A=@n96, B=@n60.RGB) {constB=0}, B=@n82.Result), B=@n88.Result)
  NamedRerouteDeclaration(None=Divide(A=Add(A=@n76, B=@n95), B=Add(A=@n97, B=@n68.RGB)) {constB=1}) {name=SampleResult, nodeColor=<rgba 0.3359 0 1 1>}
  n98 = ViewProperty() {property=MEVP_TemporalSampleIndex}
  MaterialExpressionAdd_2 = Add(A=ScreenPosition().PixelPosition, B=AppendVector(A=@n98.Property, B=@n98.Property))
  n99 = Custom!(p=@MaterialExpressionAdd_2) {
   code="Mod( ((uint)(p.x) + 2 * (uint)(p.y)) , 5 )"
   outputType=CMOT_Float1
   inputs [{inputName=p, input {expression=&self:MaterialExpressionAdd_2}}]
  }
  n100 = TextureSample(UVs=MaterialFunctionCall!(TextureResolution=Constant2Vector() {g=64, r=64}) {materialFunction=/Engine/Functions/Engine_MaterialFunctions02/Texturing/ScreenAlignedPixelToPixelUVs}.Result) {samplerType=SAMPLERTYPE_LinearGrayscale, texture=/Engine/EngineMaterials/Good64x64TilingNoiseHighFreq}
  NamedRerouteDeclaration(None=Multiply(A=Add(A=@n99, B=@n100.R) {constB=0}) {constB=0.1666, desc=Divide by 6 avoiding a precision problem where an Alpha Threshold of 0 still leaves some remaining pixels.}) {name=SampledDither, nodeColor=<rgba 0 1 0.0156 1>}
  n101 = DynamicParameter() {defaultValue=<rgba 1 1 0.7 1>, paramNames [VelocityTextureStrength, RadialBlurStrength, MaskDissolve, UseMask]}
  n102 = DynamicParameter() {
   parameterIndex=2
   defaultValue=<rgba 1 1 0.7 1>
   paramNames [VelocityTextureTime, RadialBlurStrengthTime, MaskDissolveTime, ScaleVelocityTextureUV]
  }
  n103 = DynamicParameter() {
   parameterIndex=1
   defaultValue=<rgba 0 200 1 1>
   paramNames [Desaturation, DistanceFade, VelocityTextureRotationSpeed, UseSpriteUVForVelocityTexture]
  }
  n104 = TextureObjectParameter() {
   group=VelocityTexture
   parameterName=VelocityTexture_Texture
   samplerType=SAMPLERTYPE_LinearColor
   sortPriority=0
   texture=/Game/Effects/Fort_Effects/Creatures/Elemental/Fire/MergedFireFlowMap
  }
  n105 = TextureCoordinate()
  n106 = DDX(Value=@n105)
  n107 = DDY(Value=@n105)
  n108 = VectorParameter() {
   group=VelocityTexture
   parameterName=VelocityTexture_UVControl
   channelNames {a=VPan, b=UPan, g=VScale, r=UScale}
   defaultValue=<rgba 1 0.5 0 -1.3>
  }
  n109 = DynamicParameter() {
   parameterIndex=2
   defaultValue=<rgba 1 1 0.7 1>
   paramNames [VelocityTextureTime, RadialBlurStrengthTime, MaskDissolveTime, ScaleVelocityTextureUV]
  }
  n110 = AppendVector(A=Multiply(A=AppendVector(A=@n108.UScale, B=@n108.UScale), B=@n109.ScaleVelocityTextureUV), B=@n108.UPan)
  n111 = StaticSwitchParameter(True=TextureCoordinate(), False=NamedRerouteUsage()) {parameterName=UseSpriteUVForVelocityTexture}
  n112 = AppendVector(
   A=Subtract(A=MaterialFunctionCall(
     Texture=@n104
     "[Override]DDXInput"=@n106
     "[Override]DDYInput"=@n107
     UVControl_RGB=@n110
     UVControl_A=@n108.VPan
     "[Override]TimeInput"=@n102.VelocityTextureTime
     "[Override]UVInput"=@n111
    ) {materialFunction=&FXCommon_TextureSample_DDXDDY}.RG) {constB=0.5}
   B=Constant()
  )
  n113 = RotateAboutAxis(
   NormalizedRotationAxis=Constant3Vector() {constant=<rgba 0 0 1 0>}
   RotationAngle=Multiply(A=@n102.VelocityTextureTime, B=@n103.VelocityTextureRotationSpeed)
   PivotPoint=Constant()
   Position=@n112
  )
  n114 = Distance(A=ParticlePositionWS(), B=CameraPositionWS())
  n115 = ComponentMask(None=Divide(A=Add(A=@n113, B=@n112), B=Divide(A=@n114) {constB=750})) {g=true, r=true}
  n116 = VectorParameter() {
   group=VelocityTexture
   parameterName=VelocityTexture_ToBlur
   sortPriority=60
   channelNames {a=VStrength, b=UStrength, g=VOffset, r=UOffset}
   defaultValue=<rgba 0 0 0.2 0.2>
  }
  n117 = MaterialFunctionCall(NoiseValue=@n115, Control_RGB=@n116, Control_A=@n116) {materialFunction=&FXCommon_NoiseControl}
  NamedRerouteDeclaration(None=Multiply(A=@n56.A, B=Multiply(A=@n101.VelocityTextureStrength, B=@n117.NoiseValue))) {name=VelocityTexture_ToBlur, nodeColor=<rgba 1 0.0938 0 1>}
  n118 = VectorParameter() {
   group=VelocityTexture
   parameterName=VelocityTexture_ToRefraction
   sortPriority=60
   channelNames {a=VStrength, b=UStrength, g=VOffset, r=UOffset}
   defaultValue=<rgba 0 0 0.1 0.1>
  }
  n119 = MaterialFunctionCall(NoiseValue=@n115, Control_RGB=@n118, Control_A=@n118) {materialFunction=&FXCommon_NoiseControl}
  NamedRerouteDeclaration(None=Multiply(A=@n56.A, B=Multiply(A=@n101.VelocityTextureStrength, B=@n119.NoiseValue))) {name=VelocityTexture_ToRefraction, nodeColor=<rgba 1 0.0938 0 1>}
  n120 = ParticlePositionWS()
  NamedRerouteDeclaration(None=Subtract(
    A=MaterialFunctionCall(World Position=@n120) {materialFunction=&TransformToClipSpace}.Screen aligned 0-1 UV
    B=ScreenPosition().ViewportUV
   )) {name=VectorToTheRadialCenter, nodeColor=<rgba 0 0.7891 1 1>}
  n121 = MaterialFunctionCall() {materialFunction=&ScreenResolution}
  n122 = Multiply(
   A=MaterialFunctionCall!(Center Point World Position=@n120) {materialFunction=/Engine/Functions/Engine_MaterialFunctions02/UVs/WorldSpaceAlignedScreenCoordinates}.Offset and Scaled UVS
   B=Divide(
    A=MaterialFunctionCall(Float2=@n121.Buffer Resolution) {materialFunction=&BreakOutFloat2Components}.G
    B=Constant() {r=985}
   ) {constB=20}
  )
  NamedRerouteDeclaration(None=MaterialFunctionCall(Vector or UVs=Multiply(A=@n122) {constB=2}) {materialFunction=&VectorToRadialValue}.Radial Coordinates) {name=RadialCoord, nodeColor=<rgba 1 0.8906 0 1>}
  NamedRerouteDeclaration(None=@n115) {name=VelocityTexture_OriginV, nodeColor=<rgba 0.5703 0 1 1>}
  NamedRerouteUsage()
  ScalarParameter() {defaultValue=0.7, group=mask, parameterName=mask_DissolvePower}
 ]
}
